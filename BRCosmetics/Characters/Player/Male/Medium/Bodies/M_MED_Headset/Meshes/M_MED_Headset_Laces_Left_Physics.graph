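# Physics asset for the left shoelaces of the male medium Headset character body, covering the dynamic lace bones and the left foot.
# Capsule bodies form four two-bone lace chains (A to D) that branch from a shared base bone, which is constrained to a larger foot capsule.
# Constraints give each joint tight swing limits, narrower on one axis for chain B, with the twist limit at zero throughout.
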PhysicsAsset /BRCosmetics/Characters/Player/Male/Medium/Bodies/M_MED_Headset/Meshes/M_MED_Headset_Laces_Left_Physics
bodies {
 dyn_lacesA_01_l [{shapeType=Capsule, length=1, radius=1, shapeName=dyn_lacesA_01_l_capsule, center=<0.7716 0 0>, rotation=<rot 90 0 0>}]
 dyn_lacesA_02_l [{shapeType=Capsule, length=1, radius=1, shapeName=dyn_lacesA_02_l_capsule, center=<0.7716 0 0>, rotation=<rot 90 0 0>}]
 dyn_lacesB_01_l [{shapeType=Capsule, length=1.3736, radius=1.1867, shapeName=dyn_lacesB_01_l_capsule, center=<0.9383 -0.9 0>, rotation=<rot 90 0 0>}]
 dyn_lacesB_02_l [{shapeType=Capsule, length=0.8914, radius=1.1086, shapeName=dyn_lacesB_02_l_capsule, center=<0.7353 -0.6 0>, rotation=<rot 90 0 0>}]
 dyn_lacesBase_01_l [{shapeType=Capsule, length=1, radius=1, shapeName=dyn_lacesBase_01_l_capsule}]
 dyn_lacesC_01_l [{shapeType=Capsule, length=2.7593, radius=1.2403, shapeName=dyn_lacesC_01_l_capsule, center=<1.336 -1 0>, rotation=<rot 90 0 0>}]
 dyn_lacesC_02_l [{shapeType=Capsule, length=2.9882, radius=1.2403, shapeName=dyn_lacesC_02_l_capsule, center=<0.8214 -1 0>, rotation=<rot 90 0 0>}]
 dyn_lacesD_01_l [{shapeType=Capsule, length=2.5151, radius=1.2239, shapeName=dyn_lacesD_01_l_capsule, center=<1.8915 -0.8 0>, rotation=<rot 90 0 0>}]
 dyn_lacesD_02_l [{shapeType=Capsule, length=2.8307, radius=1.2239, shapeName=dyn_lacesD_02_l_capsule, center=<0.8376 -0.8 0>, rotation=<rot 90 0 0>}]
 foot_l [{
   shapeType=Capsule
   length=14.8413
   radius=5.2069
   shapeName=foot_l_capsule
   center=<-2.8648 6.2813 0.4239>
   rotation=<rot 2.1416 35.4335 -90>
  }]
}
constraints [
 {bone1Name=dyn_lacesBase_01_l, bone2Name=foot_l, swing1LimitDegrees=20, swing2LimitDegrees=20, twistLimitDegrees=0}
 {bone1Name=dyn_lacesA_01_l, bone2Name=dyn_lacesBase_01_l, swing1LimitDegrees=20, swing2LimitDegrees=20, twistLimitDegrees=0}
 {bone1Name=dyn_lacesA_02_l, bone2Name=dyn_lacesA_01_l, swing1LimitDegrees=20, swing2LimitDegrees=20, twistLimitDegrees=0}
 {bone1Name=dyn_lacesB_01_l, bone2Name=dyn_lacesBase_01_l, swing1LimitDegrees=20, swing2LimitDegrees=10, twistLimitDegrees=0}
 {bone1Name=dyn_lacesB_02_l, bone2Name=dyn_lacesB_01_l, swing1LimitDegrees=20, swing2LimitDegrees=10, twistLimitDegrees=0}
 {bone1Name=dyn_lacesC_01_l, bone2Name=dyn_lacesBase_01_l, swing1LimitDegrees=20, swing2LimitDegrees=20, twistLimitDegrees=0}
 {bone1Name=dyn_lacesC_02_l, bone2Name=dyn_lacesC_01_l, swing1LimitDegrees=20, swing2LimitDegrees=20, twistLimitDegrees=0}
 {bone1Name=dyn_lacesD_01_l, bone2Name=dyn_lacesBase_01_l, swing1LimitDegrees=20, swing2LimitDegrees=20, twistLimitDegrees=0}
 {bone1Name=dyn_lacesD_02_l, bone2Name=dyn_lacesD_01_l, swing1LimitDegrees=20, swing2LimitDegrees=20, twistLimitDegrees=0}
]
