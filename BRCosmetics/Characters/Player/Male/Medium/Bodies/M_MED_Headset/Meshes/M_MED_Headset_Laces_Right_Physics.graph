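# Physics asset for the right-foot shoelaces of the male medium Headset body, providing collision bodies and joints for the dynamic lace bones.
# A small base body links to a large foot_r capsule with free swing and twist, and anchors four two-segment lace chains (A to D).
# Each lace joint has a narrow swing limit and zero twist limit.
# The foot_r capsule carries a left-side shape name, foot_l_capsule.
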PhysicsAsset /BRCosmetics/Characters/Player/Male/Medium/Bodies/M_MED_Headset/Meshes/M_MED_Headset_Laces_Right_Physics
bodies {
 dyn_lacesA_01_r [{shapeType=Capsule, length=2.2972, radius=1, shapeName=dyn_lacesA_01_r_capsule, center=<-1.1069 0 -0.8>, rotation=<rot 90 0 0>}]
 dyn_lacesA_02_r [{shapeType=Capsule, length=2.2972, radius=1, shapeName=dyn_lacesA_02_r_capsule, center=<-1.1069 0 -0.8>, rotation=<rot 90 0 0>}]
 dyn_lacesB_01_r [{shapeType=Capsule, length=2.2972, radius=1, shapeName=dyn_lacesB_01_r_capsule, center=<-1.1069 0.8 0>, rotation=<rot 90 0 0>}]
 dyn_lacesB_02_r [{shapeType=Capsule, length=2.2972, radius=1, shapeName=dyn_lacesB_02_r_capsule, center=<-1.1069 0.2 0>, rotation=<rot 90 90 70.6867>}]
 dyn_lacesBase_01_r [{shapeType=Capsule, length=0, radius=0.7926, shapeName=dyn_lacesBase_01_r_capsule}]
 dyn_lacesC_01_r [{shapeType=Capsule, length=2.2972, radius=1, shapeName=dyn_lacesC_01_r_capsule, center=<-1.1069 0.4778 0>, rotation=<rot 90 0 0>}]
 dyn_lacesC_02_r [{shapeType=Capsule, length=2.2972, radius=1, shapeName=dyn_lacesC_02_r_capsule, center=<-1.1069 0.8 0>, rotation=<rot 90 0 0>}]
 dyn_lacesD_01_r [{shapeType=Capsule, length=2.2972, radius=1, shapeName=dyn_lacesD_01_r_capsule, center=<-1.1069 0.6342 0>, rotation=<rot 90 0 0>}]
 dyn_lacesD_02_r [{shapeType=Capsule, length=2.2972, radius=1, shapeName=dyn_lacesD_02_r_capsule, center=<-1.1069 0.624 0>, rotation=<rot 90 0 0>}]
 foot_r [{
   shapeType=Capsule
   length=16.2263
   radius=5.1482
   shapeName=foot_l_capsule
   center=<2.8648 -6.2813 -0.4239>
   rotation=<rot 2.1416 35.4335 90>
  }]
}
constraints [
 {
  bone1Name=dyn_lacesBase_01_r
  bone2Name=foot_r
  swing1LimitDegrees=45
  swing1Motion=Free
  swing2LimitDegrees=45
  swing2Motion=Free
  twistLimitDegrees=45
  twistMotion=Free
 }
 {bone1Name=dyn_lacesA_01_r, bone2Name=dyn_lacesBase_01_r, swing1LimitDegrees=15, swing2LimitDegrees=15, twistLimitDegrees=0}
 {bone1Name=dyn_lacesA_02_r, bone2Name=dyn_lacesA_01_r, swing1LimitDegrees=15, swing2LimitDegrees=15, twistLimitDegrees=0}
 {bone1Name=dyn_lacesB_01_r, bone2Name=dyn_lacesBase_01_r, swing1LimitDegrees=15, swing2LimitDegrees=15, twistLimitDegrees=0}
 {bone1Name=dyn_lacesB_02_r, bone2Name=dyn_lacesB_01_r, swing1LimitDegrees=15, swing2LimitDegrees=15, twistLimitDegrees=0}
 {bone1Name=dyn_lacesC_01_r, bone2Name=dyn_lacesBase_01_r, swing1LimitDegrees=15, swing2LimitDegrees=15, twistLimitDegrees=0}
 {bone1Name=dyn_lacesC_02_r, bone2Name=dyn_lacesC_01_r, swing1LimitDegrees=15, swing2LimitDegrees=15, twistLimitDegrees=0}
 {bone1Name=dyn_lacesD_01_r, bone2Name=dyn_lacesBase_01_r, swing1LimitDegrees=15, swing2LimitDegrees=15, twistLimitDegrees=0}
 {bone1Name=dyn_lacesD_02_r, bone2Name=dyn_lacesD_01_r, swing1LimitDegrees=15, swing2LimitDegrees=15, twistLimitDegrees=0}
]
