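# Physics asset for the M_MED_Headset male medium player body, covering the pelvis and three dynamic shirt bone chains (center, left, right).
# The pelvis uses two large capsules, while each shirt chain has a capsule on its MSTR bone and its 01 bone.
# Each MSTR bone rotates freely against the pelvis.
# Each 01 bone is constrained to its MSTR with a narrow single-axis swing, giving a hinge-like flap.
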PhysicsAsset /BRCosmetics/Characters/Player/Male/Medium/Bodies/M_MED_Headset/Meshes/M_MED_Headset_Physics
bodies {
 dyn_shirt_c_01 [{shapeType=Capsule, length=10.2509, radius=2.7543, shapeName=dyn_shirt_c_01_capsule, center=<3.1224 -2.5 0>, rotation=<rot 90 0 0>}]
 dyn_shirt_c_MSTR [{shapeType=Capsule, length=0.505, radius=0.505, shapeName=dyn_shirt_c_MSTR_capsule, rotation=<rot 0 -90 0>}]
 dyn_shirt_l_01 [{shapeType=Capsule, length=5, radius=2.7543, shapeName=dyn_shirt_l_01_capsule, center=<2.6568 -2.5 0>, rotation=<rot 90 0 0>}]
 dyn_shirt_l_MSTR [{shapeType=Capsule, length=5.7559, radius=0.505, shapeName=dyn_shirt_l_MSTR_capsule, center=<0 0 -0.4656>, rotation=<rot 0 -90 0>}]
 dyn_shirt_r_01 [{shapeType=Capsule, length=9.8187, radius=3.1865, shapeName=dyn_shirt_r_01_capsule, center=<3.4495 -2.5 0>, rotation=<rot 90 0 0>}]
 dyn_shirt_r_MSTR [{shapeType=Capsule, length=0.505, radius=0.505, shapeName=dyn_shirt_r_MSTR_capsule, rotation=<rot 0 -90 0>}]
 pelvis [
  {shapeType=Capsule, length=10.8945, radius=11.0541, shapeName=pelvis_capsule, center=<0.609 -0.3322 -0.3348>, rotation=<rot 0 -90 0>}
  {shapeType=Capsule, length=8.4703, radius=12.7668, shapeName=pelvis_capsule, center=<-5.1036 0 0>}
 ]
}
constraints [
 {bone1Name=dyn_shirt_c_01, bone2Name=dyn_shirt_c_MSTR, swing1LimitDegrees=20, swing2LimitDegrees=0, twistLimitDegrees=0}
 {bone1Name=dyn_shirt_l_01, bone2Name=dyn_shirt_l_MSTR, swing1LimitDegrees=20, swing2LimitDegrees=0, twistLimitDegrees=0}
 {bone1Name=dyn_shirt_r_01, bone2Name=dyn_shirt_r_MSTR, swing1LimitDegrees=20, swing2LimitDegrees=0, twistLimitDegrees=0}
 {
  bone1Name=dyn_shirt_c_MSTR
  bone2Name=pelvis
  swing1LimitDegrees=45
  swing1Motion=Free
  swing2LimitDegrees=45
  swing2Motion=Free
  twistLimitDegrees=45
  twistMotion=Free
 }
 {
  bone1Name=dyn_shirt_l_MSTR
  bone2Name=pelvis
  swing1LimitDegrees=45
  swing1Motion=Free
  swing2LimitDegrees=45
  swing2Motion=Free
  twistLimitDegrees=45
  twistMotion=Free
 }
 {
  bone1Name=dyn_shirt_r_MSTR
  bone2Name=pelvis
  swing1LimitDegrees=45
  swing1Motion=Free
  swing2LimitDegrees=45
  swing2Motion=Free
  twistLimitDegrees=45
  twistMotion=Free
 }
]
