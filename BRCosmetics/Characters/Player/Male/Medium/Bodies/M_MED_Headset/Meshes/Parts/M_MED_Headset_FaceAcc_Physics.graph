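# Physics asset for the face accessory mesh of the male medium Headset character, setting up collision and joints for its dynamic hair bones.
# One head capsule anchors four two-bone chains, three left-side strands and one top piece, each built from capsule bodies.
# Each chain's MSTR root rotates freely against the head.
# The left strand tips have small swing limits and zero twist limit, while the top tip is rotationally locked to its root.
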PhysicsAsset /BRCosmetics/Characters/Player/Male/Medium/Bodies/M_MED_Headset/Meshes/Parts/M_MED_Headset_FaceAcc_Physics
bodies {
 dyn_hair_l_A_01 [{shapeType=Capsule, length=0, radius=1, shapeName=dyn_hair_l_A_01_capsule, center=<1.5348 0 0>, rotation=<rot 90 0 0>}]
 dyn_hair_l_A_MSTR [{shapeType=Capsule, length=0.5, radius=0.5, shapeName=dyn_hair_l_A_MSTR_capsule, rotation=<rot 0 -90 0>}]
 dyn_hair_l_B_01 [{shapeType=Capsule, length=0.5839, radius=1.1673, shapeName=dyn_hair_l_B_01_capsule, center=<2.0976 0 0>, rotation=<rot 90 0 0>}]
 dyn_hair_l_B_MSTR [{shapeType=Capsule, length=0.5, radius=0.5, shapeName=dyn_hair_l_B_MSTR_capsule, rotation=<rot 0 -90 0>}]
 dyn_hair_l_C_01 [{shapeType=Capsule, length=0.9697, radius=1.1682, shapeName=dyn_hair_l_C_01_capsule, center=<1.2733 0 0>, rotation=<rot 90 0 0>}]
 dyn_hair_l_C_MSTR [{shapeType=Capsule, length=0.5, radius=0.5, shapeName=dyn_hair_l_C_MSTR_capsule, rotation=<rot 0 -90 0>}]
 dyn_hair_top_01 [{shapeType=Capsule, length=1.3068, radius=1.9109, shapeName=dyn_hair_top_01_capsule, center=<0 0.817 0>, rotation=<rot 90 0 0>}]
 dyn_hair_top_MSTR [{shapeType=Capsule, length=0.5, radius=0.5, shapeName=dyn_hair_top_MSTR_capsule, rotation=<rot 0 -90 0>}]
 head [{shapeType=Capsule, length=2.668, radius=5.2789, center=<10.8058 1.3164 -0.7262>, rotation=<rot 0 -2.2197 -79.2108>}]
}
constraints [
 {
  bone1Name=dyn_hair_l_A_MSTR
  bone2Name=head
  swing1LimitDegrees=45
  swing1Motion=Free
  swing2LimitDegrees=45
  swing2Motion=Free
  twistLimitDegrees=45
  twistMotion=Free
 }
 {bone1Name=dyn_hair_l_A_01, bone2Name=dyn_hair_l_A_MSTR, swing1LimitDegrees=10, swing2LimitDegrees=8, twistLimitDegrees=0}
 {
  bone1Name=dyn_hair_l_B_MSTR
  bone2Name=head
  swing1LimitDegrees=45
  swing1Motion=Free
  swing2LimitDegrees=45
  swing2Motion=Free
  twistLimitDegrees=45
  twistMotion=Free
 }
 {bone1Name=dyn_hair_l_B_01, bone2Name=dyn_hair_l_B_MSTR, swing1LimitDegrees=20, swing2LimitDegrees=20, twistLimitDegrees=0}
 {
  bone1Name=dyn_hair_l_C_MSTR
  bone2Name=head
  swing1LimitDegrees=45
  swing1Motion=Free
  swing2LimitDegrees=45
  swing2Motion=Free
  twistLimitDegrees=45
  twistMotion=Free
 }
 {bone1Name=dyn_hair_l_C_01, bone2Name=dyn_hair_l_C_MSTR, swing1LimitDegrees=20, swing2LimitDegrees=20, twistLimitDegrees=0}
 {
  bone1Name=dyn_hair_top_MSTR
  bone2Name=head
  swing1LimitDegrees=45
  swing1Motion=Free
  swing2LimitDegrees=45
  swing2Motion=Free
  twistLimitDegrees=45
  twistMotion=Free
 }
 {
  bone1Name=dyn_hair_top_01
  bone2Name=dyn_hair_top_MSTR
  swing1LimitDegrees=45
  swing1Motion=Locked
  swing2LimitDegrees=45
  swing2Motion=Locked
  twistLimitDegrees=45
  twistMotion=Locked
 }
]
